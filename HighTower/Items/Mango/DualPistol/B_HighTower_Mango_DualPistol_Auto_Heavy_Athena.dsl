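; Blueprint for the heavy automatic dual pistol weapon in the HighTower Mango item set.
; It overrides DeactivateShellsFX to call the parent only when the right-hand spawned shell particle system is valid, guarding against a missing effect.
; The construction script only defers to the parent, so all other behavior comes from the parent class.
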(fn DeactivateShellsFX
  (Utilities|IsValid (Variables|Effects-Shells|GetParticleSystem_SpawnedShells_Right)
    (:Is%20Valid
      (|CallParentFunction))))

(fn ConstructionScript
  (|CallParentFunction))
